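# M_RR_SimpleNeonSign is a Nanite-enabled lit emissive material for neon sign props.
# A vertex-red mask blends two sign colors, with optional color-B blinking, sequenced lighting and world-position noise flicker.
# Fresnel and a shared NeonCity balance function shape the emissive, with a simpler MaterialProxy path.
# An isSpline switch adds SplineThicken widening by camera distance and a gradient mask in place of the second Fresnel falloff.
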Material /Hermes/Props/AlleyTown/RR_NeonSigns_A/Materials/M_RR_SimpleNeonSign
graph {
 properties {
  bTangentSpaceNormal=false
  bUseMaterialAttributes=true
  bUsedWithNanite=true
  pixelDepthOffsetMode=PDOM_Legacy
  usedShadingModels=Default Lit
 }
 nodes [
  Clamp()
  n1 = Divide(A=WorldPosition().XYZ, B=Constant() {r=500000})
  n2 = Multiply(A=Time(), B=Constant() {r=0.15})
  n3 = TextureSample(UVs=ComponentMask(None=Add(A=@n1, B=@n2)) {g=true, r=true}) {samplerType=SAMPLERTYPE_LinearColor, texture=/Game/Global/Textures/Default/Noise/Noise32x32}
  Floor(None=@n3.R)
  n4 = ComponentMask(None=TextureCoordinate()) {r=true}
  MaterialFunctionCall(UV Channel=@n4) {materialFunction=&LinearGradient}
  n5 = VectorParameter() {parameterName=Sign Color A, defaultValue=<rgba 1 0.0105 0 1>}
  n6 = VectorParameter() {parameterName=Sign Color B, defaultValue=<rgba 0 0.3136 1 1>}
  n7 = NamedRerouteUsage()
  n8 = LinearInterpolate(A=@n7, B=OneMinus(None=@n7), Alpha=ScalarParameter() {parameterName=INvertColorB})
  n9 = LinearInterpolate(A=@n5.RGB, B=@n6.RGB, Alpha=@n8)
  n10 = ScalarParameter() {defaultValue=1, parameterName=Sign Brightness}
  n11 = LinearInterpolate(Alpha=ScalarParameter() {defaultValue=0.5, parameterName=BiasOn}) {constB=2}
  n12 = Time()
  n13 = Frac(None=Multiply(A=Multiply(A=@n12) {constB=0.1}, B=ScalarParameter() {defaultValue=4, parameterName=Blink Rate}))
  n14 = Saturate(None=Round(None=Multiply(A=@n11, B=@n13)))
  n15 = LinearInterpolate(A=@n5.RGB, B=LinearInterpolate(A=@n6.RGB, B=Multiply(A=@n14, B=@n6.RGB), Alpha=ScalarParameter() {parameterName=Blink Color B}), Alpha=@n8)
  n16 = ScalarParameter() {defaultValue=0.5, parameterName=RadiusSequence}
  n17 = Time()
  n18 = ScalarParameter() {defaultValue=0.1, parameterName=SequenceSpeed}
  n19 = LinearInterpolate(A=Subtract(B=@n16) {constA=0}, B=Add(B=@n16) {constA=1}, Alpha=Frac(None=Multiply(A=@n17, B=@n18)))
  n20 = VertexColor()
  n21 = Round(None=Add(B=Multiply(A=@n19, B=@n20.R)))
  n22 = LinearInterpolate(Alpha=SphereMask(A=@n20.R, B=@n19, Radius=@n16) {attenuationRadius=0.2}) {constA=0.03}
  n23 = StaticSwitchParameter(True=Multiply(A=@n15, B=StaticSwitchParameter(True=@n21, False=@n22) {parameterName=AdditiveBlink}), False=@n15) {parameterName=SequenceBlink}
  n24 = LinearInterpolate(
   A=ScalarParameter() {parameterName=Brightness Min}
   B=ScalarParameter() {defaultValue=1, parameterName=Brightness Max}
   Alpha=Multiply(A=@n4, B=ScalarParameter() {defaultValue=1, parameterName=asdfasdf})
  )
  n25 = StaticSwitchParameter(True=NaniteReplace!(Default=@n23, Nanite=Multiply(A=@n23, B=@n24)), False=@n23) {parameterName=MultiplyByGreenVC}
  n26 = Multiply(A=@n3.R, B=ScalarParameter() {defaultValue=2.5, parameterName=FlickerBias}) {constB=2}
  n27 = LinearInterpolate(Alpha=Round(None=Saturate(None=@n26))) {constA=0.7}
  n28 = Multiply(A=LinearInterpolate(A=@n25, B=Multiply(A=@n25, B=@n27), Alpha=ScalarParameter() {parameterName=Flickering})) {constB=0.6}
  n29 = Fresnel(ExponentIn=ScalarParameter() {defaultValue=1, parameterName=EmissiveFresnel}) {baseReflectFraction=0, exponent=1}
  n30 = LinearInterpolate(Alpha=Saturate(None=OneMinus(None=@n29))) {constA=0.2}
  n31 = MaterialFunctionCall!(Emissive=MaterialProxyReplace(Realtime=Multiply(A=@n28, B=@n30), MaterialProxy=Multiply(A=@n28) {constB=0.5})) {materialFunction=/Game/Environments/Asteria/Sets/NeonCity/Techart/Screen/MaterialFunctions/MF_NeonCityEmissiveBalance}
  n32 = SetMaterialAttributes!(
   MaterialAttributes
   Base Color=Multiply(A=Saturate(None=@n9)) {constB=0.1}
   Emissive Color=Multiply(A=Multiply(A=@n10, B=@n31.Result), B=ScalarParameter() {defaultValue=1, parameterName=EmissiveMult})
   Specular=Constant() {r=0.6}
   Roughness=Constant() {r=0.1}
   Normal=VertexNormalWS()
   World Position Offset
  ) {attributeSetTypes [
    #1
    #2
    #3
    #4
    #5
    #6
   ]}
  n33 = GetMaterialAttributes(None=@n32) {attributeGetTypes [#2, #5]}
  n34 = Multiply(
   A=@n33.Emissive Color
   B=ConstantBiasScale(None=TextureSample() {texture=/Game/Effects/Fort_Effects/Emotes/Sienna/T_Horizontal_gradient_Mask}.R) {bias=0.5, scale=1.2}
  )
  n35 = ScalarParameter() {parameterName=Emissive Saturation}
  n36 = ScalarParameter() {defaultValue=4, group=SplineThicken, parameterName=ST_WidthBase}
  n37 = LinearInterpolate(
   A=@n36
   B=Multiply(A=@n36, B=ScalarParameter() {defaultValue=2.5, parameterName=DistanceThicknessmult}) {constB=2}
   Alpha=MaterialFunctionCall(
    Fade Length=ScalarParameter() {defaultValue=2000, parameterName=ThicknessDistance}
    Fade Offset=ScalarParameter() {defaultValue=2000, parameterName=ThicknessDistance}
    For Vertex Shader=StaticBool() {value=true}
   ) {materialFunction=&CameraDepthFade}.Result
  )
  n38 = MaterialFunctionCall(WidthBase=@n37, WidthTip=@n37, AngleCorrectedNormals=StaticBoolParameter() {group=SplineThicken, parameterName=ST_AngleCorrectedNormals}) {materialFunction=&SplineThicken}
  n39 = VertexNormalWS()
  n40 = Multiply(A=@n33.Emissive Color, B=OneMinus(None=Fresnel(Normal=@n39) {baseReflectFraction=0, exponent=1}))
  n41 = ScalarParameter() {parameterName=Emissive Saturation}
  n42 = StaticSwitchParameter(
   True=SetMaterialAttributes!(
    MaterialAttributes=@n33.MaterialAttributes
    Emissive Color=Desaturation(None=@n34, Fraction=@n35)
    Normal=@n38.Normal
    World Position Offset=@n38.Normal
   ) {attributeSetTypes [#2, #5, #6]}
   False=SetMaterialAttributes!(MaterialAttributes=@n33.MaterialAttributes, Emissive Color=Desaturation(None=@n40, Fraction=@n41)) {attributeSetTypes [#2]}
  ) {parameterName=isSpline}
  MaterialProxyReplace(Realtime=@n42, MaterialProxy=SetMaterialAttributes(MaterialAttributes=@n42) {attributeSetTypes [#2]})
  NamedRerouteDeclaration(None=SphereMask(A=VertexColor().R, B=Constant()) {attenuationRadius=0.5}) {name=ColorBMask, nodeColor=<rgba 1 0.0234 0 1>}
 ]
}
